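# Niagara system for the Mint skin of the SnowGlobe character, rendering green head fire with glow and drifting snow flecks.
# Fire, glow and top-mesh emitters follow sockets on User.SkeletalMesh every update, animating curve-driven offsets, colors and sizes, with flipbook flames.
# A separate decal-named sprite spawns in a box, while snow spawns on a hemisphere and moves by gravity, attraction and curl noise.
# Every emitter sets VisibilityTag from User.bisHeadHidden, tying visibility to head hiding.
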NiagaraSystem /BRCosmetics/Characters/Player/Male/Medium/Bodies/M_MED_SnowGlobe/Skins/Mint/VFX/NS_SnowGlobeMint_HeadFire
dependencies {
 usedDataInterfaces [&NiagaraDataInterfaceCurve, &NiagaraDataInterfaceVectorCurve, &NiagaraDataInterfaceColorCurve]
 usedDynamicInputs [&VectorFromCurve]
 usedModules [&SystemState]
}
emitters {
 Decal001 {
  emitterUpdateScript [
   &EmitterState()
   &SpawnRate(SpawnRate=20, Use Spawn Probability=false)
  ]
  particleSpawnScript [
   &FN_SetSpawnGroup_v2(Number of Spawn Groups=2)
   /Niagara/Modules/Spawn/Initialization/InitializeParticle(
    Color=<rgba 0 1 0.1656 3>
    Lifetime=0.65
    Module.Write Color=true
    Module.Write Lifetime=true
    Module.Write Mass=true
    Module.Write Material Random=true
    Module.Write Position=false
    Module.Write Scale=false
    Module.Write SpriteRotation=false
    Module.Write SpriteSize=true
    Position=&SimulationPosition()
    Sprite Size=&Vector2DFromFloat(Value=20)
   )
   &DynamicMaterialParameters(
    Index 0 Param 1=0.5
    Index 0 Param 2=1
    Index 0 Param 3=1
    Module.Param0WriteEnabled=false
    Module.Param1WriteEnabled=true
    Module.Param2WriteEnabled=true
    Module.Param3WriteEnabled=false
   )
   &BoxLocation(Box Offset=<0 0 8>, Box Size=<10 10 10>, MaskSpawnGroup=false, Module.Override Seed=false, Rotate Box=false, Surface Only=false)
  ]
  particleUpdateScript [
   &ParticleState()
   &DynamicMaterialParameters(
    Index 0 Param 1=&FloatFromCurve(FloatCurve=DataInterface() {propertyValues {Curve {keys [<at 0.4 3>,<at 1 6>]}}})
    Module.Param0WriteEnabled=true
    Module.Param1WriteEnabled=false
    Module.Param2WriteEnabled=false
    Module.Param3WriteEnabled=false
   ) {module=DynamicMaterialParameters001}
   &ScaleColor(
    Scale Alpha=&FloatFromCurve(FloatCurve=DataInterface() {propertyValues {Curve {keys [<at 0 0 RCIM_Cubic>,<at 0.2 1 RCIM_Cubic>,<at 1 0 RCIM_Cubic>]}}}, Scale Curve=0.15)
    ScaleA=true
    ScaleRGB=true
   )
   &CameraOffset(Camera Offset Amount=&RandomRangeFloat(Minimum=5, Maximum=10))
   /SetVariables(Particles.VisibilityTag=&MakeCustomIntFromBool(Bool=$User.bisHeadHidden, True INT=1, False INT=0)) {isSetParameters=true, module=SetVariables_#1}
  ]
  rendererClasses [&NiagaraSpriteRendererProperties]
  renderers [{rendererClass=&NiagaraSpriteRendererProperties}]
 }
 FireSmall {
  emitterUpdateScript [
   &EmitterState()
   &SpawnRate(SpawnRate=14, Use Spawn Probability=false)
  ]
  particleSpawnScript [
   &FN_SetSpawnGroup(Number of Spawn Groups=7)
   &InitializeParticle(
    Color=<rgba 1 1 1 0.3>
    Lifetime=&UniformRangedFloat(Minimum=0.9, Maximum=1.3)
    Module.Write Color=true
    Module.Write Lifetime=true
    Module.Write Mass=false
    Module.Write Material Random=true
    Module.Write Position=false
    Module.Write Scale=false
    Module.Write SpriteRotation=true
    Module.Write SpriteSize=true
    Position=&SimulationPosition()
    Sprite Rotation=&UniformRangedFloat(Minimum=-180, Maximum=180)
    Sprite Size=&Vector2DFromFloat(Value=&RandomRangeFloat(Minimum=9, Maximum=11))
   )
   &DynamicMaterialParameters(
    Index 0 Param 2=&UniformRangedFloat(Minimum=0, Maximum=1)
    Index 0 Param 3=1
    Index 0 Param 4=&UniformRangedFloat(Minimum=2, Maximum=2.25)
    Module.Param0WriteEnabled=false
    Module.Param1WriteEnabled=true
    Module.Param2WriteEnabled=false
    Module.Param3WriteEnabled=true
   ) {module=DynamicMaterialParameters002}
   &CameraOffset(Camera Offset Amount=&RandomRangeFloat(Minimum=6, Maximum=6))
  ]
  particleUpdateScript [
   &ParticleState()
   &SkeletalMeshLocation(
    Bone / Socket Index=&SpawnGroup()
    Inherit Velocity=false
    Interpolate Mesh Sampling=false
    MaskSpawnGroup=false
    Mesh Sampling Type=::NewEnumerator5
    Sampled Position Offset=&Lerp_Vector(
     A=<0 0 0>
     B=<3 0 0>
     Alpha=&FloatFromCurve(FloatCurve=DataInterface() {propertyValues {
        bOptimizeLUT=false
        Curve {keys [
          <at 0 0 RCIM_Cubic>
          {arriveTangent=1.502, interpMode=RCIM_Cubic, leaveTangent=1.502, tangentMode=RCTM_User, time=0.4363, value=0.5179}
          {interpMode=RCIM_Cubic, leaveTangent=2.0003, time=1, value=1}
         ]}
       }})
    )
    Sampled Position Offset Coordinate Space:ENiagaraMeshTransforms=::NewEnumerator3
    Skeletal Mesh=$User.SkeletalMesh
    Socket Sampling Mode=::NewEnumerator5
   )
   &ScaleColor(
    Scale Mode=::NewEnumerator1
    Scale RGBA=/Niagara/DynamicInputs/UniformRange/UniformRangedVector4(
     Minimum:Vector4f=&Vector4FromCurve(Vector4Curve=DataInterface() {propertyValues {
        AlphaCurve {keys [<at 0.0024 0>,<at 0.0816 0.8>,<at 0.5772 1>,<at 0.9977 -0.0026>]}
        BlueCurve {keys [<at 0.1692 0>,<at 0.2549 0>]}
        GreenCurve {keys [<at 0.1692 50>,<at 0.2549 10>]}
        RedCurve {keys [<at 0.1692 11.3189>,<at 0.2549 2.1632>]}
       }})
     Maximum:Vector4f=&Vector4FromCurve(Vector4Curve=DataInterface() {propertyValues {
        AlphaCurve {keys [<at 0.0024 0>,<at 0.077 0.8>,<at 0.5696 1>,<at 0.9977 0>]}
        BlueCurve {keys [<at 0.1718 3.4665>,<at 0.2624 0.242>]}
        GreenCurve {keys [<at 0.1718 50>,<at 0.2624 10>]}
        RedCurve {keys [<at 0.1718 0>,<at 0.2624 0>]}
       }})
     Randomness Mode:ENiagaraRandomnessMode=::NewEnumerator0
     Random Seed:NiagaraInt32=0
     Fixed Random Seed:NiagaraBool=false
    )
    ScaleRGBA=true
   ) {module=ScaleColor001}
   /SetVariables(Particles.NewGlowColor:LinearColor=&ColorFromCurve(ColorCurve=DataInterface() {propertyValues {
       AlphaCurve {keys [<at -0.0036 0>,<at 0.244 0.85 RCIM_Cubic>,<at 1 0 RCIM_Cubic>]}
       BlueCurve {keys [<at 0.2478 0.0311 RCIM_Cubic>]}
       GreenCurve {keys [<at 0.2478 0.2 RCIM_Cubic>]}
       RedCurve {keys [<at 0.2478 0 RCIM_Cubic>]}
      }})) {isSetParameters=true, module=SetVariables_#2}
   /Niagara/Modules/Update/SubUV/SubUVAnimation(Number Of Frames=54)
   &ScaleSpriteSize(Scale Factor=&Vector2DFromFloat(Value=&FloatFromCurve(FloatCurve=DataInterface() {propertyValues {Curve {keys [{arriveTangent=1.1585, interpMode=RCIM_Cubic, leaveTangent=1.1585, tangentMode=RCTM_User, time=0, value=0.4863}, <at 1 0.8738 RCIM_Cubic>]}}})))
   &DynamicMaterialParameters(
    Index 0 Param 1=&FloatFromCurve(FloatCurve=DataInterface() {propertyValues {bOptimizeLUT=false, Curve {keys [<at 0.1314 1 RCIM_Cubic>,<at 0.2911 0.5008 RCIM_Cubic>,<at 0.5761 1>]}}})
    Index 0 Param 2=1
    Index 0 Param 3=0.9
    Index 0 Param 4=1
    Module.Param0WriteEnabled=true
    Module.Param1WriteEnabled=true
    Module.Param2WriteEnabled=true
    Module.Param3WriteEnabled=false
   ) {module=DynamicMaterialParameters001}
   /SetVariables(Particles.VisibilityTag=&MakeCustomIntFromBool(Bool=$User.bisHeadHidden, True INT=1, False INT=0)) {isSetParameters=true, module=SetVariables_#1}
  ]
  rendererClasses [&NiagaraSpriteRendererProperties]
  renderers [
   {rendererClass=&NiagaraSpriteRendererProperties}
   {rendererClass=&NiagaraSpriteRendererProperties, rendererIndex=1}
  ]
 }
 FireSmallTop {
  emitterUpdateScript [
   &EmitterState()
   &SpawnRate(SpawnRate=2, Use Spawn Probability=false)
  ]
  particleSpawnScript [
   &InitializeParticle(
    Color=<rgba 1 1 1 0.6>
    Lifetime=&UniformRangedFloat(Minimum=1.75, Maximum=2.15)
    Module.Write Color=true
    Module.Write Lifetime=true
    Module.Write Mass=false
    Module.Write Material Random=true
    Module.Write Position=false
    Module.Write Scale=false
    Module.Write SpriteRotation=true
    Module.Write SpriteSize=true
    Position=&SimulationPosition()
    Sprite Rotation=&UniformRangedFloat(Minimum=-180, Maximum=180)
    Sprite Size=&Vector2DFromFloat(Value=&RandomRangeFloat(Minimum=7, Maximum=10))
   )
   &DynamicMaterialParameters(
    Index 0 Param 2=&UniformRangedFloat(Minimum=0, Maximum=1)
    Index 0 Param 3=1
    Index 0 Param 4=&UniformRangedFloat(Minimum=1, Maximum=2)
    Module.Param0WriteEnabled=false
    Module.Param1WriteEnabled=true
    Module.Param2WriteEnabled=false
    Module.Param3WriteEnabled=true
   ) {module=DynamicMaterialParameters002}
   &CameraOffset(Camera Offset Amount=1)
   /SetVariables(Particles.CameraOffset2:NiagaraFloat=&RandomRangeFloat(Minimum=0, Maximum=10)) {isSetParameters=true}
  ]
  particleUpdateScript [
   &ParticleState()
   &SkeletalMeshLocation(
    Bone / Socket Index=7
    Inherit Velocity=false
    MaskSpawnGroup=false
    Mesh Sampling Type=::NewEnumerator5
    Sampled Position Offset=&Lerp_Vector(
     A=&RandomRangeVector(Minimum=<0 0 0>, Maximum=<0 0 0>)
     B=&RandomRangeVector(Minimum=<-6 -6 0>, Maximum=<6 6 0>)
     Alpha=&FloatFromCurve(FloatCurve=DataInterface() {propertyValues {Curve {keys [<at 0 0 RCIM_Cubic>, {arriveTangent=2.0003, interpMode=RCIM_Cubic, leaveTangent=2.0003, tangentMode=RCTM_User, time=1, value=1}]}}})
    )
    Sampled Position Offset Coordinate Space:ENiagaraMeshTransforms=::NewEnumerator2
    Skeletal Mesh=$User.SkeletalMesh
    Socket Sampling Mode=::NewEnumerator5
   )
   &ScaleColor(
    Scale Mode=::NewEnumerator1
    Scale RGBA=&UniformRangedVector4(
     Minimum:Vector4f=&Vector4FromCurve(Vector4Curve=DataInterface() {propertyValues {
        AlphaCurve {keys [<at 0.0024 0>,<at 0.1271 1>,<at 0.9977 -0.0026>]}
        BlueCurve {keys [<at 0.2725 0>,<at 0.3581 0>]}
        GreenCurve {keys [<at 0.2725 30>,<at 0.3581 10>]}
        RedCurve {keys [<at 0.2725 6.7914>,<at 0.3581 2.1632>]}
       }})
     Maximum:Vector4f=&Vector4FromCurve(Vector4Curve=DataInterface() {propertyValues {
        AlphaCurve {keys [<at 0.0024 0>,<at 0.1271 1>,<at 0.9977 -0.0026>]}
        BlueCurve {keys [<at 0.27 2.0799>,<at 0.3707 0.242>]}
        GreenCurve {keys [<at 0.27 30>,<at 0.3707 10>]}
        RedCurve {keys [<at 0.27 0>,<at 0.3707 0>]}
       }})
     Randomness Mode:ENiagaraRandomnessMode=::NewEnumerator0
     Random Seed:NiagaraInt32=0
     Fixed Random Seed:NiagaraBool=false
    )
    ScaleRGBA=true
   ) {module=ScaleColor001}
   /SetVariables(Particles.NewGlowColor:LinearColor=&ColorFromCurve(ColorCurve=DataInterface() {propertyValues {
       AlphaCurve {keys [<at 0.0024 0>,<at 0.1271 0.35>,<at 0.8164 -0.0026>]}
       BlueCurve {keys [<at 0.3937 0>]}
       GreenCurve {keys [<at 0.3937 2>]}
       RedCurve {keys [<at 0.3937 0.4528>]}
      }})) {isSetParameters=true, module=SetVariables_#2}
   &SubUVAnimation(Number Of Frames=64)
   &ScaleSpriteSize(Scale Factor=&Vector2DFromFloat(Value=&FloatFromCurve(FloatCurve=DataInterface() {propertyValues {Curve {keys [
          {arriveTangent=7.5764, interpMode=RCIM_Cubic, leaveTangent=7.5765, tangentMode=RCTM_User, time=0, value=0.02}
          {arriveTangent=1.2063, interpMode=RCIM_Cubic, leaveTangent=1.2063, tangentMode=RCTM_User, time=0.18, value=0.99}
          <at 1.0003 0.4588 RCIM_Cubic>
         ]}}})))
   &DynamicMaterialParameters(
    Index 0 Param 1=&FloatFromCurve(FloatCurve=DataInterface() {propertyValues {Curve {keys [{arriveTangent=2.7551, interpMode=RCIM_Cubic, leaveTangent=2.7551, tangentMode=RCTM_User, time=0, value=0}, <at 0.7 1>]}}})
    Index 0 Param 2=1
    Index 0 Param 3=1
    Index 0 Param 4=1
    Module.Param0WriteEnabled=true
    Module.Param1WriteEnabled=true
    Module.Param2WriteEnabled=true
    Module.Param3WriteEnabled=false
   ) {module=DynamicMaterialParameters001}
   /SetVariables(Particles.VisibilityTag=&MakeCustomIntFromBool(Bool=$User.bisHeadHidden, True INT=1, False INT=0)) {isSetParameters=true, module=SetVariables_#1}
  ]
  rendererClasses [&NiagaraSpriteRendererProperties]
  renderers [
   {rendererClass=&NiagaraSpriteRendererProperties}
   {rendererClass=&NiagaraSpriteRendererProperties, rendererIndex=1}
  ]
 }
 Fire_Back {
  emitterUpdateScript [
   &EmitterState()
   &SpawnRate(SpawnRate=10, Use Spawn Probability=false)
  ]
  particleSpawnScript [
   &InitializeParticle(
    Color=<rgba 1 1 1 0.05>
    Lifetime=&UniformRangedFloat(Minimum=2.15, Maximum=2.75)
    Module.Write Color=true
    Module.Write Lifetime=true
    Module.Write Mass=false
    Module.Write Material Random=true
    Module.Write Position=false
    Module.Write Scale=false
    Module.Write SpriteRotation=true
    Module.Write SpriteSize=true
    Position=&SimulationPosition()
    Sprite Rotation=&UniformRangedFloat(Minimum=-180, Maximum=180)
    Sprite Size=&Vector2DFromFloat(Value=&RandomRangeFloat(Minimum=25, Maximum=30))
   )
   &DynamicMaterialParameters(
    Index 0 Param 2=&UniformRangedFloat(Minimum=0, Maximum=1)
    Index 0 Param 3=1
    Index 0 Param 4=&UniformRangedFloat(Minimum=2, Maximum=3)
    Module.Param0WriteEnabled=false
    Module.Param1WriteEnabled=true
    Module.Param2WriteEnabled=false
    Module.Param3WriteEnabled=true
   ) {module=DynamicMaterialParameters002}
   &CameraOffset(Camera Offset Amount=&RandomRangeFloat(Minimum=-10, Maximum=7))
  ]
  particleUpdateScript [
   &ParticleState()
   &SkeletalMeshLocation(
    Bone / Socket Index=8
    Inherit Velocity=false
    MaskSpawnGroup=false
    Mesh Sampling Type=::NewEnumerator5
    Sampled Position Offset=&Lerp_Vector(
     A=&RandomRangeVector(Minimum=<-2 -2 0>, Maximum=<2 2 0>)
     B=<0 0 18>
     Alpha=&FloatFromCurve(FloatCurve=DataInterface() {propertyValues {Curve {keys [
          {arriveTangent=0.119, interpMode=RCIM_Cubic, leaveTangent=0.119, tangentMode=RCTM_User, time=0, value=0}
          {arriveTangent=0.3276, interpMode=RCIM_Cubic, leaveTangent=0.3276, tangentMode=RCTM_User, time=1, value=1}
         ]}}})
    )
    Sampled Position Offset Coordinate Space:ENiagaraMeshTransforms=::NewEnumerator3
    Skeletal Mesh=$User.SkeletalMesh
    Socket Sampling Mode=::NewEnumerator5
   )
   &ScaleColor(
    Scale Mode=::NewEnumerator1
    Scale RGBA=&Vector4FromCurve(Vector4Curve=DataInterface() {propertyValues {
       AlphaCurve {keys [<at 0.0024 0>,<at 0.1271 1>,<at 0.9977 -0.0026>]}
       BlueCurve {keys [<at 0.3937 0>,<at 0.75 0>]}
       GreenCurve {keys [<at 0.3937 10>,<at 0.75 1>]}
       RedCurve {keys [<at 0.3937 2.2638>,<at 0.75 0.2163>]}
      }})
    ScaleRGBA=true
   ) {module=ScaleColor001}
   &SubUVAnimation(Number Of Frames=64)
   &ScaleSpriteSize(Scale Factor=&Vector2DFromFloat(Value=&FloatFromCurve(FloatCurve=DataInterface() {propertyValues {Curve {keys [
          {arriveTangent=7.5764, interpMode=RCIM_Cubic, leaveTangent=7.5765, tangentMode=RCTM_User, time=0, value=0.02}
          {arriveTangent=1.2063, interpMode=RCIM_Cubic, leaveTangent=1.2063, tangentMode=RCTM_User, time=0.18, value=0.99}
          <at 1 0.5 RCIM_Cubic>
         ]}}})))
   &DynamicMaterialParameters(
    Index 0 Param 1=&FloatFromCurve(FloatCurve=DataInterface() {propertyValues {Curve {keys [{arriveTangent=2.7551, interpMode=RCIM_Cubic, time=0, value=1}, <at 0.7 1>]}}})
    Index 0 Param 2=1
    Index 0 Param 3=1
    Index 0 Param 4=1
    Module.Param0WriteEnabled=true
    Module.Param1WriteEnabled=true
    Module.Param2WriteEnabled=true
    Module.Param3WriteEnabled=false
   ) {module=DynamicMaterialParameters001}
   /SetVariables(Particles.VisibilityTag=&MakeCustomIntFromBool(Bool=$User.bisHeadHidden, True INT=1, False INT=0)) {isSetParameters=true, module=SetVariables_#1}
  ]
  rendererClasses [&NiagaraSpriteRendererProperties]
  renderers [{rendererClass=&NiagaraSpriteRendererProperties}]
 }
 Glow {
  emitterUpdateScript [
   &EmitterState()
   &SpawnRate(SpawnRate=10, Use Spawn Probability=false)
  ]
  particleSpawnScript [
   &InitializeParticle(
    Color=<rgba 5 5 5 0.05>
    Lifetime=&UniformRangedFloat(Minimum=1, Maximum=1.75)
    Module.Write Color=true
    Module.Write Lifetime=true
    Module.Write Mass=false
    Module.Write Material Random=true
    Module.Write Position=false
    Module.Write Scale=false
    Module.Write SpriteRotation=true
    Module.Write SpriteSize=true
    Position=&SimulationPosition()
    Sprite Rotation=&UniformRangedFloat(Minimum=-180, Maximum=180)
    Sprite Size=&Vector2DFromFloat(Value=&UniformRangedFloat(Minimum=10, Maximum=15))
   )
   &CameraOffset(Camera Offset Amount=&RandomRangeFloat(Minimum=-5, Maximum=10))
  ]
  particleUpdateScript [
   &ParticleState()
   &ScaleColor(
    Scale Mode=::NewEnumerator1
    Scale RGBA=&UniformRangedVector4(
     Minimum:Vector4f=&Vector4FromCurve(Vector4Curve=DataInterface() {propertyValues {
        AlphaCurve {keys [<at 0.0024 0>,<at 0.1271 0.8>,<at 0.265 0.5>,<at 0.9977 -0.0026>]}
        BlueCurve {keys [<at 0.3937 0>]}
        GreenCurve {keys [<at 0.3937 1>]}
        RedCurve {keys [<at 0.3937 0.2264>]}
       }})
     Maximum:Vector4f=&Vector4FromCurve(Vector4Curve=DataInterface() {propertyValues {
        AlphaCurve {keys [<at 0.0024 0>,<at 0.1271 0.5>,<at 0.265 0.5>,<at 0.9977 -0.0026>]}
        BlueCurve {keys [<at 0.3937 0.5689>]}
        GreenCurve {keys [<at 0.3937 1>]}
        RedCurve {keys [<at 0.3937 0>]}
       }})
     Randomness Mode:ENiagaraRandomnessMode=::NewEnumerator0
     Random Seed:NiagaraInt32=0
     Fixed Random Seed:NiagaraBool=false
    )
    ScaleRGBA=true
   ) {module=ScaleColor001}
   &ScaleSpriteSize(Scale Factor=&Vector2DFromFloat(Value=&FloatFromCurve(FloatCurve=DataInterface() {propertyValues {Curve {keys [
          {arriveTangent=7.5764, interpMode=RCIM_Cubic, leaveTangent=7.5765, tangentMode=RCTM_User, time=0, value=0.02}
          {arriveTangent=1.2063, interpMode=RCIM_Cubic, leaveTangent=1.2063, tangentMode=RCTM_User, time=0.18, value=0.99}
          <at 1 0.5 RCIM_Cubic>
         ]}}})))
   &SkeletalMeshLocation(
    Bone / Socket Index=8
    Inherit Velocity=false
    MaskSpawnGroup=false
    Mesh Sampling Type=::NewEnumerator5
    Sampled Position Offset=&Lerp_Vector(
     A=&RandomRangeVector(Minimum=<-4 -4 -4>, Maximum=<4 4 4>)
     B=<0 0 20>
     Alpha=&FloatFromCurve(FloatCurve=DataInterface() {propertyValues {Curve {keys [
          {arriveTangent=0.119, interpMode=RCIM_Cubic, leaveTangent=0.119, tangentMode=RCTM_User, time=0, value=0}
          {arriveTangent=0.3276, interpMode=RCIM_Cubic, leaveTangent=0.3276, tangentMode=RCTM_User, time=1, value=1}
         ]}}})
    )
    Sampled Position Offset Coordinate Space:ENiagaraMeshTransforms=::NewEnumerator3
    Skeletal Mesh=$User.SkeletalMesh
    Socket Sampling Mode=::NewEnumerator5
   )
   /SetVariables(Particles.VisibilityTag=&MakeCustomIntFromBool(Bool=$User.bisHeadHidden, True INT=1, False INT=0)) {isSetParameters=true, module=SetVariables_#1}
  ]
  rendererClasses [&NiagaraSpriteRendererProperties]
  renderers [{rendererClass=&NiagaraSpriteRendererProperties}]
 }
 NE_FireTopModel {
  emitterUpdateScript [
   &EmitterState(Life Cycle Mode=::NewEnumerator1, Loop Duration=1, UseLoopDelay=false)
   &SpawnRate(SpawnRate=2, Use Spawn Probability=false)
  ]
  particleSpawnScript [&InitializeParticle(
    Color=<rgba 1 1 1 0.2>
    Lifetime=5
    Mesh Scale=<0.2 0.2 0.3>
    Module.Write Color=true
    Module.Write Lifetime=true
    Module.Write Mass=true
    Module.Write Material Random=true
    Module.Write Position=false
    Module.Write Scale=true
    Module.Write SpriteRotation=false
    Module.Write SpriteSize=false
    Position=&SimulationPosition()
    Sprite Size=<5 5>
   )]
  particleUpdateScript [
   &ParticleState()
   &SkeletalMeshLocation(
    Apply Sampled Orientation to Mesh Particles=true
    Bone / Socket Index=9
    Inherit Velocity=false
    Interpolate Mesh Sampling=false
    MaskSpawnGroup=false
    Mesh Sampling Type=::NewEnumerator5
    Sampled Position Offset=<-0.2 0 0>
    Sampled Position Offset Coordinate Space:ENiagaraMeshTransforms=::NewEnumerator2
    Skeletal Mesh=$User.SkeletalMesh
    Socket Sampling Mode=::NewEnumerator5
   )
   /SetVariables(Particles.VisibilityTag=&MakeCustomIntFromBool(Bool=$User.bisHeadHidden, True INT=1, False INT=0)) {isSetParameters=true, module=SetVariables_#1}
  ]
  rendererClasses [&NiagaraMeshRendererProperties]
  renderers [{rendererClass=&NiagaraMeshRendererProperties, rendererIndex=0}]
 }
 NE_Snow {
  emitterUpdateScript [
   &EmitterState()
   &SpawnRate(SpawnRate=32, Use Spawn Probability=false)
  ]
  particleSpawnScript [
   /Niagara/Modules/Spawn/Initialization/V2/InitializeParticle(
    Color Mode=::NewEnumerator0
    Lifetime Max=4
    Lifetime Min=3
    Lifetime Mode=::NewEnumerator1
    Sprite Size Mode=::NewEnumerator1
    Uniform Sprite Size Max=1.8
    Uniform Sprite Size Min=1.4
    UsePositionOffset=false
    Write Material Random=true
   )
   /SetVariables(Particles.SpriteRotationRate=&RandomRangeFloat(Minimum=-0.25, Maximum=0.25, Evaluation Type=::NewEnumerator1)) {isSetParameters=true}
   &AddVelocity(Velocity=&RandomRangeVector(Minimum=<-10 -10 -10>, Maximum=<10 10 10>, Evaluation Type=::NewEnumerator1))
   /Niagara/Modules/Spawn/Location/SphereLocation(Hemisphere Z=true, MaskSpawnGroup=false, Module.Override Randomness=false, Offset=<0 0 10>, Sphere Radius=12, Surface Only=true)
  ]
  particleUpdateScript [
   &ParticleState()
   &Drag(Drag=10, Use Linear Drag=true, Use Rotational Drag=false)
   &AccelerationForce(Acceleration=<0 0 -25>)
   &PointAttractionForce(
    Attraction Strength=&FloatFromCurve(FloatCurve=DataInterface() {propertyValues {Curve {keys [<at 0 0>,<at 5 15>]}}}, CurveIndex=$Particles.Age)
    Kill Within Radius=false
    Module.Use Falloff=true
    UseAttractorOffset=false
   )
   &CurlNoiseForce(Noise Frequency=250, Noise Strength=25, Pan Noise=true, Pan Noise Field=<0 0 0.25>)
   &SolveForcesAndVelocity(Clamp Velocity=false, Limit Acceleration=false)
   &SpriteRotationRate(Rotation Rate=$Particles.SpriteRotationRate)
   &ScaleSpriteSize(Scale Factor=&Vector2DFromFloat(Value=&FloatFromCurve(FloatCurve=DataInterface() {propertyValues {Curve {keys [
          {arriveTangent=0.9932, time=0, value=0}
          {arriveTangent=-0.7835, leaveTangent=-0.7835, time=0.2002, value=0.9932}
          {arriveTangent=0.7991, leaveTangent=0.7991, time=0.3821, value=0.2097}
          {arriveTangent=-1.0088, leaveTangent=-1.0088, time=0.5993, value=1.0088}
          <at 1 0>
         ]}}})))
   &Color(Color=&MakeLinearColorFromVectorAndFloat(
     "Vector (RGB)"=&MakeVectorFromLinearColorRGB(Color=<rgba 1 1 1 1>)
     "Float (Alpha)"=&ClampFloat(Float=&FloatFromCurve(FloatCurve=DataInterface() {propertyValues {Curve {keys [{arriveTangent=-1, time=0, value=1}, <at 1 0>]}}}))
    ))
   &ScaleColor(Scale Alpha=0.15, ScaleA=true, ScaleRGB=false)
   &CameraOffset(Camera Offset Amount=&RandomRangeFloat(Minimum=-5, Maximum=15))
   /SetVariables(Particles.VisibilityTag=&MakeCustomIntFromBool(Bool=$User.bisHeadHidden, True INT=1, False INT=0)) {isSetParameters=true, module=SetVariables_#1}
  ]
  rendererClasses [&NiagaraSpriteRendererProperties]
  renderers [{rendererClass=&NiagaraSpriteRendererProperties}]
 }
}
summary {
 systemName=NS_SnowGlobeMint_HeadFire
 emitters [
  {emitterName=Fire_Back, rendererClasses [&NiagaraSpriteRendererProperties]}
  {emitterName=Decal001, rendererClasses [&NiagaraSpriteRendererProperties]}
  {emitterName=Glow, rendererClasses [&NiagaraSpriteRendererProperties]}
  {emitterName=FireSmall, rendererClasses [&NiagaraSpriteRendererProperties]}
  {emitterName=NE_FireTopModel, rendererClasses [&NiagaraMeshRendererProperties]}
  {emitterName=FireSmallTop, rendererClasses [&NiagaraSpriteRendererProperties]}
  {emitterName=NE_Snow, rendererClasses [&NiagaraSpriteRendererProperties]}
 ]
 userVariables [
  User.bisHeadHidden {type=NiagaraBool}
  User.SkeletalMesh {
   type=NiagaraDataInterfaceSkeletalMesh
   defaultValue {
    struct=NiagaraExt_VariableValue_DataInterface
    value {dataInterface=&self:NiagaraDataInterfaceSkeletalMesh_0}
   }
  }
 ]
}
